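# Niagara master example system, going by its name and folder, for unlit SubUV sprite effects, built from a single sprite emitter.
# The emitter spawns particles at a steady rate, with varied size, hue-shifted bright orange color and a per-particle material random value.
# SubUV Animation reads the sprite renderer to pick flipbook frames within a limited range.
# Emitter spawn also sets camera-distance shrink attributes, with no consumer shown.
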NiagaraSystem /Game/NeverCook/FX/Niagara_FX_Reuse_Examples/Spine/NS_SubUV_Unlit_Master
dependencies {
 usedDataInterfaces [&NiagaraDataInterfaceSpriteRendererInfo]
 usedDynamicInputs []
 usedModules [&SystemState]
}
emitters {Empty {
  emitterSpawnScript [
   /SetVariables(Emitter.Sprite Shrink Camera Distance=5) {isSetParameters=true}
   /SetVariables(Emitter.Sprite Shrink Distance Falloff Hardness=5) {isSetParameters=true}
  ]
  emitterUpdateScript [
   &EmitterState(Life Cycle Mode=::NewEnumerator1, Loop Duration=1, UseLoopDelay=false)
   &SpawnRate(SpawnRate=5, Use Spawn Probability=false)
  ]
  particleSpawnScript [&InitializeParticle(
    AdjustAlpha=false
    AdjustHue=true
    AdjustSaturation=false
    AdjustValue=false
    Color=<rgba 5 1.3206 0 1>
    Color Mode=::NewEnumerator3
    Hue Shift Range:Vector2f=<-0.1 1>
    Sprite Size Mode=::NewEnumerator1
    Uniform Sprite Size Max=100
    Uniform Sprite Size Min=50
    UsePositionOffset=false
    Write Material Random=true
   )]
  particleUpdateScript [
   &ParticleState()
   &SubUVAnimation(
    End Frame Range Override=3
    "Sprite Renderer "=DataInterface() {propertyValues {SpriteRenderer=&self:Empty.NiagaraSpriteRendererProperties_0}}
    SubUV Animation Mode:ENiagara_SubUVLookupModeV2=::NewEnumerator2
    UseEndFrame=true
    UseStartFrame=true
   )
  ]
  rendererClasses [&NiagaraSpriteRendererProperties]
  renderers [{rendererClass=&NiagaraSpriteRendererProperties}]
 }}
summary {
 systemName=NS_SubUV_Unlit_Master
 emitters [{emitterName=Empty, rendererClasses [&NiagaraSpriteRendererProperties]}]
 userVariables []
}
